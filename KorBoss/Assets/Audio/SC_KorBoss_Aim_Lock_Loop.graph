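# KorBoss aim lock sound cue that plays an aim start, a lock sound and a sustained targeting-beam loop.
# A mixer plays the start wave at once alongside a delayed branch mixing the lock wave with a further-delayed loop.
# The loop reuses a security turret camera beam sound, pitched up by a modulator and faded in by an enveloper.
# It uses long cone attenuation and a small concurrency cap that stops the quietest instance.
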SoundCue /KorBoss/Assets/Audio/SC_KorBoss_Aim_Lock_Loop
root=Mixer_0
nodes [
 {id=Delay_0, children [Enveloper_0], properties {delayMax=0.2, delayMin=0.2}}
 {id=Delay_1, children [Mixer_1], properties {delayMax=0.4, delayMin=0.4}}
 {
  id=Enveloper_0
  children [Modulator_0]
  properties {pitchCurve {editorCurveData {keys [{time=0}]}}, volumeCurve {editorCurveData {keys [<at 0.0903 0.0004>,<at 0.3196 0.7616>,<at 1 1>]}}}
 }
 {id=Mixer_0, children [WavePlayer_1,Delay_1], properties {inputVolume [0.65,1]}}
 {id=Mixer_1, children [WavePlayer_0,Delay_0], properties {inputVolume [1,0.7]}}
 {id=Modulator_0, children [WavePlayer_6], properties {pitchMax=2, pitchMin=2, volumeMax=1, volumeMin=1}}
 {id=WavePlayer_0, properties {soundWaveAssetPtr=&/KorBoss_Aim_Lock_01}}
 {id=WavePlayer_1, properties {soundWaveAssetPtr=&/KorBoss_Aim_Start_01}}
 {id=WavePlayer_6, properties {bLooping=true, soundWaveAssetPtr=/Game/Athena/Sounds/Gadgets/SecurityTurret/Security_Camera_TargetBeam_Loop_01}}
]
properties {
 attenuationSettings=&/Att_KorBoss_Aim_Lock_Loop_Cone_Long
 bDebug=true
 bOverrideConcurrency=true
 maxDistance=4250
 soundClassObject=&Fort_Gadgets
 volumeMultiplier=1
 attenuationOverrides {
  bEnableFocusInterpolation=true
  bEnableListenerFocus=true
  falloffDistance=1200
  nonFocusVolumeAttenuation=0.5
  stereoSpread=500
  attenuationShapeExtents {x=0}
 }
 concurrencyOverrides {maxCount=3, resolutionRule=StopQuietest}
}
