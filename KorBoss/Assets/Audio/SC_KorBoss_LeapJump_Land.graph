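# Sound cue for the Kor boss's leap-jump landing, routed to the boss-fight SFX class with its own long-range attenuation.
# A mixer layers five randomized sets: beam-attack hits, sparks, synth whooshes, leap land and leap-jump land, with several sources borrowed from MaskMaker audio.
# A quality-level root gives the first two levels the full variation pools and the last two a mixer trimmed to two waves per layer.
# It references a shared footstep-landing concurrency set.
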SoundCue /KorBoss/Assets/Audio/SC_KorBoss_LeapJump_Land
root=QualityLevel_0
nodes [
 {id=Mixer_1, children [Random_10,Random_20,Random_3,Random_1,Random_18], properties {inputVolume [1,0.75,1,0.75,1]}}
 {id=Mixer_3, children [Random_14,Random_21,Random_4,Random_8,Random_19], properties {inputVolume [1,0.75,1,0.75,1]}}
 {id=QualityLevel_0, children [Mixer_1,Mixer_1,Mixer_3,Mixer_3]}
 {id=Random_1, children [WavePlayer_2,WavePlayer_5,WavePlayer_6], properties {weights [1,1,1]}}
 {
  id=Random_10
  children [WavePlayer_35, WavePlayer_36, WavePlayer_37, WavePlayer_38, WavePlayer_39, WavePlayer_40]
  properties {weights [1,1,1,1,1,1]}
 }
 {id=Random_14, children [WavePlayer_50, WavePlayer_51], properties {weights [1,1]}}
 {id=Random_18, children [WavePlayer_67, WavePlayer_68, WavePlayer_69], properties {weights [1,1,1]}}
 {id=Random_19, children [WavePlayer_70, WavePlayer_71], properties {weights [1,1]}}
 {id=Random_20, children [WavePlayer_73, WavePlayer_74, WavePlayer_75], properties {weights [1,1,1]}}
 {id=Random_21, children [WavePlayer_76, WavePlayer_77], properties {weights [1,1]}}
 {
  id=Random_3
  children [WavePlayer_19, WavePlayer_20, WavePlayer_22, WavePlayer_23, WavePlayer_27, WavePlayer_29]
  properties {weights [1,1,1,1,1,1]}
 }
 {id=Random_4, children [WavePlayer_30, WavePlayer_31], properties {weights [1,1]}}
 {id=Random_8, children [WavePlayer_7, WavePlayer_14], properties {weights [1,1]}}
 {id=WavePlayer_14, properties {soundWaveAssetPtr=&/KorBoss_LeapLand_02}}
 {id=WavePlayer_19, properties {soundWaveAssetPtr=&/MaskMakerBoss_SynthWhooshA_01}}
 {id=WavePlayer_2, properties {soundWaveAssetPtr=&/KorBoss_LeapLand_01}}
 {id=WavePlayer_20, properties {soundWaveAssetPtr=&/MaskMakerBoss_SynthWhooshA_02}}
 {id=WavePlayer_22, properties {soundWaveAssetPtr=&/MaskMakerBoss_SynthWhooshA_03}}
 {id=WavePlayer_23, properties {soundWaveAssetPtr=&/MaskMakerBoss_SynthWhooshA_04}}
 {id=WavePlayer_27, properties {soundWaveAssetPtr=&/MaskMakerBoss_SynthWhooshA_05}}
 {id=WavePlayer_29, properties {soundWaveAssetPtr=&/MaskMakerBoss_SynthWhooshA_06}}
 {id=WavePlayer_30, properties {soundWaveAssetPtr=&MaskMakerBoss_SynthWhooshA_01}}
 {id=WavePlayer_31, properties {soundWaveAssetPtr=&MaskMakerBoss_SynthWhooshA_02}}
 {id=WavePlayer_35, properties {soundWaveAssetPtr=&/MaskMakerBoss_BeamAttack_HitB_06}}
 {id=WavePlayer_36, properties {soundWaveAssetPtr=&/MaskMakerBoss_BeamAttack_HitB_05}}
 {id=WavePlayer_37, properties {soundWaveAssetPtr=&/MaskMakerBoss_BeamAttack_HitB_04}}
 {id=WavePlayer_38, properties {soundWaveAssetPtr=&/MaskMakerBoss_BeamAttack_HitB_03}}
 {id=WavePlayer_39, properties {soundWaveAssetPtr=&/MaskMakerBoss_BeamAttack_HitB_02}}
 {id=WavePlayer_40, properties {soundWaveAssetPtr=&/MaskMakerBoss_BeamAttack_HitB_01}}
 {id=WavePlayer_5, properties {soundWaveAssetPtr=&KorBoss_LeapLand_02}}
 {id=WavePlayer_50, properties {soundWaveAssetPtr=&MaskMakerBoss_BeamAttack_HitB_06}}
 {id=WavePlayer_51, properties {soundWaveAssetPtr=&MaskMakerBoss_BeamAttack_HitB_05}}
 {id=WavePlayer_6, properties {soundWaveAssetPtr=&/KorBoss_LeapLand_03}}
 {id=WavePlayer_67, properties {soundWaveAssetPtr=&/KorBoss_LeapJump_Land_01}}
 {id=WavePlayer_68, properties {soundWaveAssetPtr=&/KorBoss_LeapJump_Land_02}}
 {id=WavePlayer_69, properties {soundWaveAssetPtr=&/KorBoss_LeapJump_Land_03}}
 {id=WavePlayer_7, properties {soundWaveAssetPtr=&KorBoss_LeapLand_01}}
 {id=WavePlayer_70, properties {soundWaveAssetPtr=&KorBoss_LeapJump_Land_01}}
 {id=WavePlayer_71, properties {soundWaveAssetPtr=&KorBoss_LeapJump_Land_02}}
 {id=WavePlayer_73, properties {soundWaveAssetPtr=&/MaskMaker_Sparks_01}}
 {id=WavePlayer_74, properties {soundWaveAssetPtr=&/MaskMaker_Sparks_02}}
 {id=WavePlayer_75, properties {soundWaveAssetPtr=&/MaskMaker_Sparks_03}}
 {id=WavePlayer_76, properties {soundWaveAssetPtr=&MaskMaker_Sparks_01}}
 {id=WavePlayer_77, properties {soundWaveAssetPtr=&MaskMaker_Sparks_02}}
]
properties {
 attenuationSettings=&/Att_KorBoss_LeapJump_Land
 maxDistance=28000
 soundClassObject=&Fort_BossFight_SFX
 volumeMultiplier=1.3
 attenuationOverrides {falloffDistance=8000, attenuationShapeExtents {x=8000}}
 concurrencyOverrides {maxCount=4, resolutionRule=PreventNew}
 concurrencySet [/Game/Sounds/Concurrency_Rules/Athena_Footstep_Land_Conc]
}
