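# Looping Sound Cue for the KorBoss leap-jump, routed to the boss-fight SFX sound class.
# A Random node picks one of two equally weighted looping waves, which feeds an Enveloper that fades volume in over a quarter second.
# It borrows the leap-jump landing attenuation and the Athena footstep-land concurrency set, with boosted cue volume.
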SoundCue /KorBoss/Assets/Audio/SC_KorBoss_LeapJump_Loop
root=Enveloper_0
nodes [
 {
  id=Enveloper_0
  children [Random_1]
  properties {pitchCurve {editorCurveData {keys [{time=0}]}}, volumeCurve {editorCurveData {keys [<at 0 0>,<at 0.25 1>]}}}
 }
 {id=Random_1, children [WavePlayer_2,WavePlayer_5], properties {weights [1,1]}}
 {id=WavePlayer_2, properties {bLooping=true, soundWaveAssetPtr=&/KorBoss_LeapJump_Loop_02}}
 {id=WavePlayer_5, properties {bLooping=true, soundWaveAssetPtr=&/KorBoss_LeapJump_Loop_01}}
]
properties {
 attenuationSettings=&/Att_KorBoss_LeapJump_Land
 maxDistance=28000
 soundClassObject=&Fort_BossFight_SFX
 volumeMultiplier=2
 attenuationOverrides {falloffDistance=8000, attenuationShapeExtents {x=8000}}
 concurrencyOverrides {maxCount=4, resolutionRule=PreventNew}
 concurrencySet [/Game/Sounds/Concurrency_Rules/Athena_Footstep_Land_Conc]
}
