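# Sound cue for the start of the Kor boss's leap jump, using the Land attenuation asset and the boss-fight SFX class.
# Each mixer layers dirt foley, a briefly delayed synth whoosh, two leap-start layers, a borrowed MaskMaker beam-shot and quiet GiftedTraveler sparks.
# Random nodes vary every layer except the single LeapStartA wave.
# A quality node gives the first two levels full variation pools and the last two a mixer trimmed to two waves per pool.
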SoundCue /KorBoss/Assets/Audio/SC_KorBoss_LeapJump_Start
root=QualityLevel_0
nodes [
 {id=Delay_0, children [Random_1], properties {delayMax=0.15, delayMin=0.15}}
 {id=Delay_1, children [Random_4], properties {delayMax=0.15, delayMin=0.15}}
 {id=Mixer_0, children [Random_2,Delay_0,WavePlayer_0,Random_0,Random_7,Random_14], properties {inputVolume [1,1,1.25,1.5,1,0.25]}}
 {id=Mixer_1, children [Random_3, Delay_1, WavePlayer_14, Random_5, Random_9, Random_12], properties {inputVolume [1,1,1.25,1.5,1,0.25]}}
 {id=QualityLevel_0, children [Mixer_0,Mixer_0,Mixer_1,Mixer_1]}
 {id=Random_0, children [WavePlayer_1, WavePlayer_12, WavePlayer_13], properties {weights [1,1,1]}}
 {
  id=Random_1
  children [WavePlayer_16, WavePlayer_17, WavePlayer_18, WavePlayer_19, WavePlayer_20, WavePlayer_21]
  properties {weights [1,1,1,1,1,1]}
 }
 {id=Random_12, children [WavePlayer_60, WavePlayer_61], properties {weights [1,1]}}
 {id=Random_14, children [WavePlayer_55, WavePlayer_56, WavePlayer_57, WavePlayer_58, WavePlayer_59], properties {weights [1,1,1,1,1]}}
 {id=Random_2, children [WavePlayer_3,WavePlayer_4,WavePlayer_5,WavePlayer_6,WavePlayer_7,WavePlayer_8], properties {weights [1,1,1,1,1,1]}}
 {id=Random_3, children [WavePlayer_10, WavePlayer_11], properties {weights [1,1]}}
 {id=Random_4, children [WavePlayer_22, WavePlayer_23], properties {weights [1,1]}}
 {id=Random_5, children [WavePlayer_15, WavePlayer_24], properties {weights [1,1]}}
 {id=Random_7, children [WavePlayer_31, WavePlayer_32, WavePlayer_33, WavePlayer_34, WavePlayer_35], properties {weights [1,1,1,1,1]}}
 {id=Random_9, children [WavePlayer_36, WavePlayer_37], properties {weights [1,1]}}
 {id=WavePlayer_0, properties {soundWaveAssetPtr=&/KorBoss_LeapStartA}}
 {id=WavePlayer_1, properties {soundWaveAssetPtr=&/KorBoss_LeapStartB_01}}
 {id=WavePlayer_10, properties {soundWaveAssetPtr=&/KorBoss_LeapJump_Foley_Dirt_06}}
 {id=WavePlayer_11, properties {soundWaveAssetPtr=&/KorBoss_LeapJump_Foley_Dirt_05}}
 {id=WavePlayer_12, properties {soundWaveAssetPtr=&/KorBoss_LeapStartB_02}}
 {id=WavePlayer_13, properties {soundWaveAssetPtr=&/KorBoss_LeapStartB_03}}
 {id=WavePlayer_14, properties {soundWaveAssetPtr=&KorBoss_LeapStartA}}
 {id=WavePlayer_15, properties {soundWaveAssetPtr=&KorBoss_LeapStartB_01}}
 {id=WavePlayer_16, properties {soundWaveAssetPtr=&/KorBoss_SynthWhooshB_01}}
 {id=WavePlayer_17, properties {soundWaveAssetPtr=&/KorBoss_SynthWhooshB_02}}
 {id=WavePlayer_18, properties {soundWaveAssetPtr=&/KorBoss_SynthWhooshB_03}}
 {id=WavePlayer_19, properties {soundWaveAssetPtr=&/KorBoss_SynthWhooshB_04}}
 {id=WavePlayer_20, properties {soundWaveAssetPtr=&/KorBoss_SynthWhooshB_05}}
 {id=WavePlayer_21, properties {soundWaveAssetPtr=&/KorBoss_SynthWhooshB_06}}
 {id=WavePlayer_22, properties {soundWaveAssetPtr=&KorBoss_SynthWhooshB_01}}
 {id=WavePlayer_23, properties {soundWaveAssetPtr=&KorBoss_SynthWhooshB_02}}
 {id=WavePlayer_24, properties {soundWaveAssetPtr=&KorBoss_LeapStartB_02}}
 {id=WavePlayer_3, properties {soundWaveAssetPtr=&KorBoss_LeapJump_Foley_Dirt_06}}
 {id=WavePlayer_31, properties {soundWaveAssetPtr=&/MaskMakerBoss_BeamAttack_ShootB_01}}
 {id=WavePlayer_32, properties {soundWaveAssetPtr=&/MaskMakerBoss_BeamAttack_ShootB_02}}
 {id=WavePlayer_33, properties {soundWaveAssetPtr=&/MaskMakerBoss_BeamAttack_ShootB_03}}
 {id=WavePlayer_34, properties {soundWaveAssetPtr=&/MaskMakerBoss_BeamAttack_ShootB_04}}
 {id=WavePlayer_35, properties {soundWaveAssetPtr=&/MaskMakerBoss_BeamAttack_ShootB_05}}
 {id=WavePlayer_36, properties {soundWaveAssetPtr=&MaskMakerBoss_BeamAttack_ShootB_01}}
 {id=WavePlayer_37, properties {soundWaveAssetPtr=&MaskMakerBoss_BeamAttack_ShootB_02}}
 {id=WavePlayer_4, properties {soundWaveAssetPtr=&KorBoss_LeapJump_Foley_Dirt_05}}
 {id=WavePlayer_5, properties {soundWaveAssetPtr=&/KorBoss_LeapJump_Foley_Dirt_04}}
 {id=WavePlayer_55, properties {soundWaveAssetPtr=&/GiftedTraveler_Montage_Sparks_C_05}}
 {id=WavePlayer_56, properties {soundWaveAssetPtr=&/GiftedTraveler_Montage_Sparks_C_04}}
 {id=WavePlayer_57, properties {soundWaveAssetPtr=&/GiftedTraveler_Montage_Sparks_C_03}}
 {id=WavePlayer_58, properties {soundWaveAssetPtr=&/GiftedTraveler_Montage_Sparks_C_02}}
 {id=WavePlayer_59, properties {soundWaveAssetPtr=&/GiftedTraveler_Montage_Sparks_C_01}}
 {id=WavePlayer_6, properties {soundWaveAssetPtr=&/KorBoss_LeapJump_Foley_Dirt_03}}
 {id=WavePlayer_60, properties {soundWaveAssetPtr=&GiftedTraveler_Montage_Sparks_C_05}}
 {id=WavePlayer_61, properties {soundWaveAssetPtr=&GiftedTraveler_Montage_Sparks_C_04}}
 {id=WavePlayer_7, properties {soundWaveAssetPtr=&/KorBoss_LeapJump_Foley_Dirt_02}}
 {id=WavePlayer_8, properties {soundWaveAssetPtr=&/KorBoss_LeapJump_Foley_Dirt_01}}
]
properties {
 attenuationSettings=&/Att_KorBoss_LeapJump_Land
 maxDistance=28000
 soundClassObject=&Fort_BossFight_SFX
 volumeMultiplier=1.3
 attenuationOverrides {falloffDistance=8000, attenuationShapeExtents {x=8000}}
 concurrencyOverrides {bLimitToOwner=true, maxCount=1, resolutionRule=PreventNew}
}
