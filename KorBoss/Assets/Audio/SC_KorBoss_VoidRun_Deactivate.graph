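# Sound cue for the KorBoss Void Run deactivation sound, randomly picking one of two equally weighted FirePetal waves.
# Nested branches on IsLocal and IsEnemy select full volume for local, slightly lower for enemy and lowest for non-enemy, with an unset IsEnemy using the enemy path.
# All three paths share the random node, each through its own mixer and attenuation node referencing KorBoss on/off settings, while cue-level attenuation references the FirePetal asset.
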SoundCue /KorBoss/Assets/Audio/SC_KorBoss_VoidRun_Deactivate
root=Branch_0
nodes [
 {id=Attenuation_0, children [Mixer_0], properties {attenuationSettings=&/Att_KorBoss_VoidRun_OnOff}}
 {id=Attenuation_1, children [Mixer_1], properties {attenuationSettings=&Att_KorBoss_VoidRun_OnOff}}
 {id=Attenuation_2, children [Mixer_2], properties {attenuationSettings=&Att_KorBoss_VoidRun_OnOff}}
 {id=Branch_0, children [Attenuation_0, Branch_1, Branch_1], properties {boolParameterName=IsLocal}}
 {id=Branch_1, children [Attenuation_1, Attenuation_2, Attenuation_1], properties {boolParameterName=IsEnemy}}
 {id=Mixer_0, children [Random_1], properties {inputVolume [1]}}
 {id=Mixer_1, children [Random_1], properties {inputVolume [0.9]}}
 {id=Mixer_2, children [Random_1], properties {inputVolume [0.8]}}
 {id=Random_1, children [WavePlayer_0,WavePlayer_4], properties {weights [1,1]}}
 {id=WavePlayer_0, properties {soundWaveAssetPtr=&/FirePetal_VoidRun_Deactivate_01}}
 {id=WavePlayer_4, properties {soundWaveAssetPtr=&/FirePetal_VoidRun_Deactivate_03}}
]
properties {
 attenuationSettings=/FirePetalBoonsGameplay/Assets/Audio/Settings/Att_FirePetal_VoidRun_OnOff
 maxDistance=6300
 soundClassObject=&Fort_Gadgets_Items
 volumeMultiplier=1
}
