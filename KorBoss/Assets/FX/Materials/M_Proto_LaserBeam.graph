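# Prototype unlit additive material for a laser beam, set up for Niagara sprites and ribbons.
# The beam is procedural, built from V-coordinate math: a centre-weighted profile multiplies fine static bands and broader bands that scroll over time, with no textures.
# The boosted result is multiplied by particle colour and alpha.
# A DynamicParameter node declares UVOffset and three placeholder channels but feeds nothing.
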Material /KorBoss/Assets/FX/Materials/M_Proto_LaserBeam
graph {
 properties {
  bAutomaticallySetUsageInEditor=false
  bUsedWithNiagaraRibbons=true
  bUsedWithNiagaraSprites=true
  blendMode=BLEND_Additive
  shadingModel=MSM_Unlit
  usedShadingModels=Unlit
 }
 nodes [
  DynamicParameter() {paramNames [UVOffset,Param2,Param3,Param4]}
  n1 = ParticleColor()
  n2 = TextureCoordinate()
  n3 = Multiply(A=ComponentMask(None=@n2) {g=true}, B=OneMinus(None=ComponentMask(None=@n2) {g=true}))
  n4 = Add(A=SmoothStep(Value=@n3) {constMax=0.5, constMin=0.2, constValue=1}, B=Power(Base=@n3))
  n5 = ComponentMask(None=Frac(None=Multiply(A=@n2) {constB=16})) {g=true}
  n6 = Multiply(A=@n5, B=OneMinus(None=@n5))
  n7 = Add(A=@n2, B=Multiply(A=Time()) {constB=0.05})
  n8 = ComponentMask(None=Frac(None=Multiply(A=@n7) {constB=8})) {g=true}
  n9 = ComponentMask(None=Frac(None=Multiply(A=@n7) {constB=4})) {g=true}
  n10 = Subtract(A=Multiply(A=@n8, B=OneMinus(None=@n8)), B=Multiply(A=@n9, B=OneMinus(None=@n9)))
  n11 = Multiply(A=Multiply(A=Multiply(A=@n4, B=@n6), B=Add(A=@n10) {constB=0.5})) {constB=50}
  Multiply(A=@n1.RGB, B=@n11)
  Multiply(A=@n1.A, B=@n11)
 ]
}
