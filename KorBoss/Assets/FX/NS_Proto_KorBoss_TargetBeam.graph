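# Prototype Niagara targeting beam for KorBoss: a glowing blue ribbon that lasts a few seconds.
# A single emitter bursts two particles through the Niagara beam modules, ending at an absolute System.TargetEnd position, alongside an exposed User.TargetEnd parameter and a system-spawn Set Variables step.
# Over particle life, curves widen the beam from zero and fade alpha in and out, while an HDR color multiplier supplies the glow.
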NiagaraSystem /KorBoss/Assets/FX/NS_Proto_KorBoss_TargetBeam
dependencies {
 usedDataInterfaces [&NiagaraDataInterfaceCurve]
 usedDynamicInputs []
 usedModules [
  &SystemState
  &self:SystemSpawnScript.SystemScriptSource.SystemScriptGraph.NiagaraNodeAssignment_0.SetVariables_#1
 ]
}
emitters {Em_Proto_KorBoss_TargetBeam {
  emitterUpdateScript [
   &EmitterState()
   &SpawnBurst_Instantaneous(Spawn Count=2, Use Spawn Probability=false, UseLoopCountLimit=false)
   /Niagara/Modules/Beams/BeamEmitterSetup(Absolute Beam End=true, Beam End:NiagaraPosition=&Add_Vector(A=$System.TargetEnd, B=<0 0 0>))
  ]
  particleSpawnScript [
   &InitializeParticle(
    Color Mode=::NewEnumerator0
    Lifetime=2.8
    Ribbon Width=50
    Ribbon Width Mode=::NewEnumerator1
    UsePositionOffset=false
    Write Material Random=true
   )
   /Niagara/Modules/Beams/SpawnBeam()
  ]
  particleUpdateScript [
   &ParticleState()
   /Niagara/Modules/Beams/ScaleBeamWidth(Beam Width Scale:NiagaraFloat=&FloatFromCurve(
     FloatCurve=DataInterface() {propertyValues {Curve {keys [<at 0 0 RCIM_Cubic>, {arriveTangent=1, leaveTangent=1, time=0.35, value=1}, <at 1 1>]}}}
     Scale Curve=0.2
    ))
   /Niagara/Modules/Beams/UpdateBeam()
   &ScaleColor(
    Scale Alpha=&FloatFromCurve(FloatCurve=DataInterface() {propertyValues {Curve {keys [<at 0 0>,<at 0.1 1>,<at 0.9 1>,<at 1 0>]}}})
    Scale RGB=&MakeVectorFromLinearColorRGB(Color=<rgba 0.8001 6.9991 100 1>)
    ScaleA=true
    ScaleRGB=true
   )
  ]
  rendererClasses [&NiagaraRibbonRendererProperties]
  renderers [{rendererClass=&NiagaraRibbonRendererProperties}]
 }}
summary {
 systemName=NS_Proto_KorBoss_TargetBeam
 emitters [{emitterName=Em_Proto_KorBoss_TargetBeam, rendererClasses [&NiagaraRibbonRendererProperties]}]
 userVariables [User.TargetEnd {type=NiagaraPosition, description="", defaultValue {}}]
}
